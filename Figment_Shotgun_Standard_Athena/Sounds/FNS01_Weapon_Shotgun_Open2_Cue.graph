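# Sound cue for a shotgun pump-open foley sound in the Figment standard shotgun's reload audio.
# A single wave player feeds the shared Fortnite shotgun pump-open sample through a modulator at the root.
# It plays in the weapon foley sound class with weapon reload attenuation and overrides concurrency to prevent new instances once the limit is reached.
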SoundCue /Figment_Shotgun_Standard_Athena/Sounds/FNS01_Weapon_Shotgun_Open2_Cue
root=Modulator_0
nodes [
 {id=Modulator_0, children [WavePlayer_10]}
 {id=WavePlayer_10, properties {soundWaveAssetPtr=/Game/Sounds/Fort_Weapons/Shotgun/Reload_Elements/weapon_shotgun_pump_open}}
]
properties {
 attenuationSettings=&Fort_Weapon_Reload_Att
 bOverrideConcurrency=true
 maxDistance=2250
 soundClassObject=&Fort_WeaponFoley
 volumeMultiplier=1
 attenuationOverrides {falloffDistance=1600}
 concurrencyOverrides {resolutionRule=PreventNew}
}
