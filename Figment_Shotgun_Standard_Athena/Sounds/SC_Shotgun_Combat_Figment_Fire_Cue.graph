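# Sound cue for the Figment combat shotgun's fire, layering close, medium and distant gunshot samples by range.
# A quality switch plays a mix with three random waves per layer at the first level and two at the others.
# Each layer has its own volume modulation, attenuation and sound class (third-person close, distant otherwise), with pitch variation only on distant layers.
# Concurrency sets reference local and global gunfire limits.
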SoundCue /Figment_Shotgun_Standard_Athena/Sounds/SC_Shotgun_Combat_Figment_Fire_Cue
root=QualityLevel_0
nodes [
 {
  id=Attenuation_1
  children [Modulator_1]
  properties {attenuationSettings=&Fort_Weapon_Projectile_Medium_Mono_Athena_Att}
 }
 {id=Attenuation_2, children [Modulator_7], properties {attenuationSettings=&Fort_Weapon_Projectile_Close_Athena_Att}}
 {
  id=Attenuation_3
  children [Modulator_6]
  properties {attenuationSettings=&Fort_Weapon_Projectile_Distant_Mono_Athena_Att}
 }
 {id=Attenuation_4, children [Modulator_0], properties {attenuationSettings=&Fort_Weapon_Projectile_Close_Athena_Att}}
 {
  id=Attenuation_5
  children [Modulator_12]
  properties {attenuationSettings=&Fort_Weapon_Projectile_Distant_Mono_Athena_Att}
 }
 {
  id=Attenuation_6
  children [Modulator_3]
  properties {attenuationSettings=&Fort_Weapon_Projectile_Medium_Mono_Athena_Att}
 }
 {id=Mixer_1, children [SoundClass_4,SoundClass_1,SoundClass_5], properties {inputVolume [1,0.8,0.8]}}
 {id=Mixer_3, children [SoundClass_2,SoundClass_6,SoundClass_3], properties {inputVolume [1,0.8,0.8]}}
 {id=Modulator_0, children [Random_1], properties {volumeMax=1.3, volumeMin=1.25}}
 {id=Modulator_1, children [Random_8], properties {volumeMax=2, volumeMin=2}}
 {id=Modulator_12, children [Random_0], properties {pitchMax=0.9, pitchMin=0.8, volumeMax=1, volumeMin=0.9}}
 {id=Modulator_3, children [Random_7], properties {volumeMax=1.15, volumeMin=1.1}}
 {id=Modulator_6, children [Random_6], properties {pitchMax=1, pitchMin=0.9, volumeMax=2, volumeMin=2}}
 {id=Modulator_7, children [Random_5], properties {volumeMax=1.3, volumeMin=1.25}}
 {id=QualityLevel_0, children [Mixer_3,Mixer_1,Mixer_1,Mixer_1]}
 {id=Random_0, children [WavePlayer_46, WavePlayer_47], properties {weights [1,1]}}
 {id=Random_1, children [WavePlayer_42, WavePlayer_43], properties {weights [1,1]}}
 {id=Random_5, children [WavePlayer_6,WavePlayer_7,WavePlayer_8], properties {weights [1,1,1]}}
 {id=Random_6, children [WavePlayer_9, WavePlayer_10, WavePlayer_11], properties {weights [1,1,1]}}
 {id=Random_7, children [WavePlayer_44, WavePlayer_45], properties {weights [1,1]}}
 {id=Random_8, children [WavePlayer_12, WavePlayer_13, WavePlayer_14], properties {weights [1,1,1]}}
 {
  id=SoundClass_1
  children [Attenuation_6]
  properties {soundClassOverride=&Fort_WeaponRanged_Fire_Distant}
 }
 {
  id=SoundClass_2
  children [Attenuation_2]
  properties {soundClassOverride=&Fort_WeaponRanged_Fire_3P_PSM}
 }
 {
  id=SoundClass_3
  children [Attenuation_3]
  properties {soundClassOverride=&Fort_WeaponRanged_Fire_Distant}
 }
 {
  id=SoundClass_4
  children [Attenuation_4]
  properties {soundClassOverride=&Fort_WeaponRanged_Fire_3P_PSM}
 }
 {
  id=SoundClass_5
  children [Attenuation_5]
  properties {soundClassOverride=&Fort_WeaponRanged_Fire_Distant}
 }
 {
  id=SoundClass_6
  children [Attenuation_1]
  properties {soundClassOverride=&Fort_WeaponRanged_Fire_Distant}
 }
 {id=WavePlayer_10, properties {soundWaveAssetPtr=&/Shotgun_Combat_Fire_DistantNew_02}}
 {id=WavePlayer_11, properties {soundWaveAssetPtr=&/Shotgun_Combat_Fire_DistantNew_03}}
 {id=WavePlayer_12, properties {soundWaveAssetPtr=&/Shotgun_Combat_Fire_MediumNew_01}}
 {id=WavePlayer_13, properties {soundWaveAssetPtr=&/Shotgun_Combat_Fire_MediumNew_02}}
 {id=WavePlayer_14, properties {soundWaveAssetPtr=&/Shotgun_Combat_Fire_MediumNew_03}}
 {id=WavePlayer_42, properties {soundWaveAssetPtr=&/Shotgun_Combat_Fire_CloseNew_01}}
 {id=WavePlayer_43, properties {soundWaveAssetPtr=&/Shotgun_Combat_Fire_CloseNew_02}}
 {id=WavePlayer_44, properties {soundWaveAssetPtr=&Shotgun_Combat_Fire_MediumNew_01}}
 {id=WavePlayer_45, properties {soundWaveAssetPtr=&Shotgun_Combat_Fire_MediumNew_02}}
 {id=WavePlayer_46, properties {soundWaveAssetPtr=&/Shotgun_Combat_Fire_DistantNew_01}}
 {id=WavePlayer_47, properties {soundWaveAssetPtr=&Shotgun_Combat_Fire_DistantNew_02}}
 {id=WavePlayer_6, properties {soundWaveAssetPtr=&Shotgun_Combat_Fire_CloseNew_01}}
 {id=WavePlayer_7, properties {soundWaveAssetPtr=&Shotgun_Combat_Fire_CloseNew_02}}
 {id=WavePlayer_8, properties {soundWaveAssetPtr=&/Shotgun_Combat_Fire_CloseNew_03}}
 {id=WavePlayer_9, properties {soundWaveAssetPtr=&Shotgun_Combat_Fire_DistantNew_01}}
]
properties {
 maxDistance=27500
 volumeMultiplier=1.05
 attenuationOverrides {falloffDistance=5000, attenuationShapeExtents {x=1000}}
 concurrencyOverrides {bLimitToOwner=true, maxCount=3}
 concurrencySet [&Weapon_Gunfire_Local_Conc, &Weapon_Gunfire_Global_Conc]
}
