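# Sound cue for the first-person firing sound of the Figment combat shotgun in Athena.
# A quality switch routes its first level to a random pick of three close-fire waves and the other three levels to a shared random pick of two.
# Each branch passes through a modulator that holds volume steady.
# First-person attenuation, a 1P weapon-fire sound class and owner-limited gunfire concurrency govern playback.
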SoundCue /Figment_Shotgun_Standard_Athena/Sounds/SC_Shotgun_Combat_Fire_Figment_1P_Cue
root=QualityLevel_0
nodes [
 {id=Modulator_1, children [Random_2], properties {volumeMax=1, volumeMin=1}}
 {id=Modulator_7, children [Random_0], properties {volumeMax=1, volumeMin=1}}
 {id=QualityLevel_0, children [Modulator_7,Modulator_1,Modulator_1,Modulator_1]}
 {id=Random_0, children [WavePlayer_0,WavePlayer_1,WavePlayer_2], properties {weights [1,1,1]}}
 {id=Random_2, children [WavePlayer_5,WavePlayer_6], properties {weights [1,1]}}
 {id=WavePlayer_0, properties {soundWaveAssetPtr=&/Shotgun_Combat_Fire_CloseNew_01}}
 {id=WavePlayer_1, properties {soundWaveAssetPtr=&/Shotgun_Combat_Fire_CloseNew_02}}
 {id=WavePlayer_2, properties {soundWaveAssetPtr=&/Shotgun_Combat_Fire_CloseNew_03}}
 {id=WavePlayer_5, properties {soundWaveAssetPtr=&Shotgun_Combat_Fire_CloseNew_01}}
 {id=WavePlayer_6, properties {soundWaveAssetPtr=&Shotgun_Combat_Fire_CloseNew_02}}
]
properties {
 attenuationSettings=&Fort_Weapon_Projectile_FirstPerson_Athena_Att
 maxDistance=2097152
 soundClassObject=&Fort_WeaponRanged_Fire_1P_PSM
 volumeMultiplier=1.1
 attenuationOverrides {falloffDistance=5000, attenuationShapeExtents {x=1000}}
 concurrencyOverrides {bLimitToOwner=true, maxCount=3}
 concurrencySet [&Weapon_Gunfire_FirstPerson_Conc]
}
