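# Sound cue for the equip foley of the high-tier Figment standard combat shotgun, reusing the pump shotgun reload wave.
# A short fixed delay feeds a default Modulator, which plays the single wave player at lowered cue pitch and volume.
# It uses weapon reload attenuation and the weapon foley sound class, with overridden concurrency that limits instances per owner and stops the oldest.
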SoundCue /Figment_Shotgun_Standard_Athena/Sounds/SC_Shotgun_Standard_Figment_Combat_HighTier_Equip
root=Delay_0
nodes [
 {id=Delay_0, children [Modulator_14], properties {delayMax=0.1, delayMin=0.1}}
 {id=Modulator_14, children [WavePlayer_1]}
 {id=WavePlayer_1, properties {soundWaveAssetPtr=/Game/Sounds/Fort_Weapons/Shotgun/Reload_Elements/weapon_shotgun_pump_reload}}
]
properties {
 attenuationSettings=&Fort_Weapon_Reload_Att
 bOverrideConcurrency=true
 maxDistance=2250
 pitchMultiplier=0.85
 soundClassObject=&Fort_WeaponFoley
 volumeMultiplier=0.587
 attenuationOverrides {falloffDistance=1600}
 concurrencyOverrides {bLimitToOwner=true, maxCount=2, resolutionRule=StopOldest}
}
