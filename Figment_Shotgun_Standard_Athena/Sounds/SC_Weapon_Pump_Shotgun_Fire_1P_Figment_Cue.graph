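# Sound cue for the first-person fire sound of the Figment standard pump shotgun.
# A Random node picks one of two equally weighted fire waves and is excluded from branch culling, feeding a Modulator that varies pitch and volume slightly.
# The cue references first-person weapon attenuation, a ranged-fire sound class and a first-person gunfire concurrency set.
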SoundCue /Figment_Shotgun_Standard_Athena/Sounds/SC_Weapon_Pump_Shotgun_Fire_1P_Figment_Cue
root=Modulator_0
nodes [
 {id=Modulator_0, children [Random_0], properties {pitchMin=1}}
 {id=Random_0, children [WavePlayer_0,WavePlayer_1], properties {bShouldExcludeFromBranchCulling=true, weights [1,1]}}
 {id=WavePlayer_0, properties {soundWaveAssetPtr=&/Shotgun_Pump_NewFire2_01}}
 {id=WavePlayer_1, properties {soundWaveAssetPtr=&/Shotgun_Pump_NewFire2_02}}
]
properties {
 attenuationSettings=&Fort_Weapon_Projectile_FirstPerson_Athena_Att
 maxDistance=2097152
 soundClassObject=&Fort_WeaponRanged_Fire_1P_PSM
 volumeMultiplier=1.1
 attenuationOverrides {falloffDistance=5000, attenuationShapeExtents {x=1000}}
 concurrencyOverrides {bLimitToOwner=true, maxCount=3, resolutionRule=StopOldest}
 concurrencySet [&Weapon_Gunfire_FirstPerson_Conc]
}
