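# Sound cue for the pump-action foley of the Figment standard shotgun.
# A single pump wave feeds a Modulator that boosts volume at fixed pitch with no random variation.
# It plays in the weapon foley sound class using shared weapon reload attenuation.
# Overridden concurrency keeps a small per-owner voice limit, stopping the oldest instance first.
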SoundCue /Figment_Shotgun_Standard_Athena/Sounds/SC_Weapon_Shotgun_Figment_Pump_Pump_Cue
root=Modulator_0
nodes [
 {id=Modulator_0, children [WavePlayer_1], properties {pitchMax=1, pitchMin=1, volumeMax=2, volumeMin=2}}
 {id=WavePlayer_1, properties {soundWaveAssetPtr=&/Pump_Shotgun_PumpNew}}
]
properties {
 attenuationSettings=&Fort_Weapon_Reload_Att
 bOverrideConcurrency=true
 maxDistance=2250
 soundClassObject=&Fort_WeaponFoley
 volumeMultiplier=0.7
 attenuationOverrides {falloffDistance=1600}
 concurrencyOverrides {bLimitToOwner=true, maxCount=2, resolutionRule=StopOldest}
}
