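# Sound cue for the single shell insert foley of the Figment pump shotgun reload.
# A Random node picks one of three shell-insert waves with equal weight, then a Modulator varies pitch slightly and applies a fixed volume boost.
# It uses weapon reload attenuation and the weapon foley sound class, with a concurrency override that prevents new instances once the limit is reached.
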SoundCue /Figment_Shotgun_Standard_Athena/Sounds/SC_Weapon_Shotgun_Pump_InsertSingle_Figment_Cue
root=Modulator_13
nodes [
 {id=Modulator_13, children [Random_1], properties {pitchMax=1.1, pitchMin=0.9, volumeMax=2, volumeMin=2}}
 {id=Random_1, children [WavePlayer_3,WavePlayer_4,WavePlayer_5], properties {weights [1,1,1]}}
 {id=WavePlayer_3, properties {soundWaveAssetPtr=&/Pump_Shotgun_InsertShellNew_03}}
 {id=WavePlayer_4, properties {soundWaveAssetPtr=&/Pump_Shotgun_InsertShellNew_02}}
 {id=WavePlayer_5, properties {soundWaveAssetPtr=&/Pump_Shotgun_InsertShellNew_01}}
]
properties {
 attenuationSettings=&Fort_Weapon_Reload_Att
 bOverrideConcurrency=true
 maxDistance=2250
 soundClassObject=&Fort_WeaponFoley
 volumeMultiplier=0.7
 attenuationOverrides {falloffDistance=1600}
 concurrencyOverrides {resolutionRule=PreventNew}
}
