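# Sound cue for the start of the Figment combat shotgun's reload, classed as weapon foley.
# A single wave player plays the shared Tactical_Reload_Start recording through a default Modulator, giving slight random pitch and volume variation.
# Playback uses reduced volume and the shared weapon reload attenuation asset.
# A concurrency override blocks new instances once the limit is reached, rather than stopping playing ones.
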SoundCue /Figment_Shotgun_Standard_Athena/Sounds/Weapon_Shotgun_Combat_ReloadStart_Figment_Cue
root=Modulator_0
nodes [
 {id=Modulator_0, children [WavePlayer_0]}
 {id=WavePlayer_0, properties {soundWaveAssetPtr=/Game/Sounds/Fort_Weapons/Shotgun/Reload_Elements/Tactical_Reload_Start}}
]
properties {
 attenuationSettings=&Fort_Weapon_Reload_Att
 bOverrideConcurrency=true
 maxDistance=2250
 soundClassObject=&Fort_WeaponFoley
 volumeMultiplier=0.702
 attenuationOverrides {falloffDistance=1600}
 concurrencyOverrides {resolutionRule=PreventNew}
}
